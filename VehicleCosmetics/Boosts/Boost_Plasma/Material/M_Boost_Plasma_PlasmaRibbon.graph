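# Translucent unlit Niagara ribbon material for the plasma vehicle boost effect.
# Three panning samples of a cloud noise texture combine and are shaped by a UV gradient mask that fades the ribbon edges.
# The result blends particle color, scaled by dynamic Brightness into an emissive adjustment function.
# It also forms particle-alpha-weighted and Distortion-scaled masks; SoftAmount and NoiseAmount are declared but unused in the visible graph.
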Material /VehicleCosmetics/Boosts/Boost_Plasma/Material/M_Boost_Plasma_PlasmaRibbon
graph {
 properties {
  bAutomaticallySetUsageInEditor=false
  bUsedWithNiagaraRibbons=true
  blendMode=BLEND_Translucent
  shadingModel=MSM_Unlit
  usedShadingModels=Unlit
 }
 nodes [
  n1 = TextureSample(UVs=Panner(Coordinate=TextureCoordinate() {uTiling=0.25, vTiling=0.3}) {speedX=-0.5, speedY=0.1}) {texture=/DealershipSharedLibrary/FX/Textures/Noise/T_VCVFX__NoiseCloud}
  n2 = TextureCoordinate() {uTiling=0.15, vTiling=0.6}
  n3 = Power(
   Base=TextureSample(UVs=Panner(Coordinate=@n2) {speedX=-1, speedY=0.25}) {texture=&T_VCVFX__NoiseCloud}.R
   Exp=Constant() {r=0.5}
  )
  n4 = Multiply(A=@n1.R, B=Constant() {r=1})
  n5 = TextureSample(UVs=Panner(Coordinate=@n2) {speedX=-0.8, speedY=0.1}) {texture=&T_VCVFX__NoiseCloud}
  n6 = Multiply(A=Multiply(A=Add(A=@n1.G, B=@n3), B=Add(A=@n4, B=@n5.B)), B=Constant() {r=5})
  n7 = ComponentMask(None=TextureCoordinate()) {r=true}
  n8 = Power(Base=OneMinus(None=@n7), Exp=Constant() {r=2})
  n9 = ComponentMask(None=TextureCoordinate()) {g=true}
  n10 = Power(Base=OneMinus(None=@n9), Exp=Constant() {r=1.5})
  n11 = Power(Base=@n9, Exp=Constant() {r=1.5})
  n12 = Multiply(A=Multiply(A=Multiply(A=@n8, B=@n7), B=Multiply(A=@n10, B=@n11)), B=Constant() {r=60})
  n13 = Multiply(A=Multiply(A=@n6, B=@n12), B=@n12)
  n14 = DynamicParameter() {paramNames [Distortion,Brightness,SoftAmount,NoiseAmount]}
  MaterialExpressionConstantClamp_1 = Saturate(None=Multiply(A=@n13, B=@n14.Distortion))
  MaterialExpressionVertexColor_0 = ParticleColor()
  MaterialExpressionConstantClamp_4 = Saturate(None=@n13)
  n15 = Multiply(A=@n6, B=@MaterialExpressionConstantClamp_4)
  n16 = Constant() {r=3}
  MaterialExpressionConstantClamp_2 = Saturate(None=Multiply(A=@MaterialExpressionVertexColor_0.A, B=Multiply(A=@n15, B=@n16)))
  MaterialExpressionConstantClamp_3 = Saturate(None=Multiply(A=Constant() {r=-0.2}, B=@MaterialExpressionVertexColor_0.RGB))
  MaterialFunctionCall(Emissive=Multiply(A=LinearInterpolate(A=@MaterialExpressionConstantClamp_3, B=@MaterialExpressionVertexColor_0.RGB, Alpha=@n15), B=@n14.Brightness)) {materialFunction=&MF_FX_EmissiveAdjustments}
 ]
}
